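# Physics asset for the LocalZilla two-handed pickaxe, giving its handle and three-bone dynamic spine simplified capsule bodies.
# The spine bones form a constraint chain from the handle, with capsule radius growing toward the last bone.
# Angular limits are tight and narrow along the chain, allowing slight flex at the first two joints, while the final joint locks swing and twist.
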PhysicsAsset /BRCosmetics/Weapons/FORT_Melee/Pickaxe_LocalZilla/Meshes/LocalZilla_Axe_Physics
bodies {
 Melee_TwoHanded_Handle [{shapeType=Capsule, length=66.991, radius=4.1273, center=<-0.0105 -1.3728 7.2974>, rotation=<rot -0.1547 -89.9998 0.0019>}]
 dyn_spine1 [{shapeType=Capsule, length=1.25, radius=4.5, shapeName=dyn_spine1_capsule, center=<6.5061 0 0>, rotation=<rot 90 180 -179.9996>}]
 dyn_spine2 [{shapeType=Capsule, length=5, radius=6.75, shapeName=dyn_spine2_capsule, center=<8.6368 0.2252 -0.0866>, rotation=<rot 80 -90.0004 -90>}]
 dyn_spine3 [{shapeType=Capsule, length=5.5, radius=9.5, shapeName=dyn_spine3_capsule, center=<12.1711 -1.7975 0>, rotation=<rot -90 180 169.9991>}]
}
constraints [
 {bone1Name=dyn_spine1, bone2Name=Melee_TwoHanded_Handle, swing1LimitDegrees=5, swing2LimitDegrees=5, twistLimitDegrees=5}
 {bone1Name=dyn_spine2, bone2Name=dyn_spine1, swing1LimitDegrees=3, swing2LimitDegrees=3, twistLimitDegrees=3}
 {
  bone1Name=dyn_spine3
  bone2Name=dyn_spine2
  swing1LimitDegrees=0.5
  swing1Motion=Locked
  swing2LimitDegrees=0.5
  swing2Motion=Locked
  twistLimitDegrees=1
  twistMotion=Locked
 }
]
